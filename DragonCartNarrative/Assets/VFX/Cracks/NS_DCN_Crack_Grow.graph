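# Niagara system for a red rift crack that grows with the User.Stage float.
# One emitter spawns three groups of mesh and sprite layers, split by visibility tag, with Stage-sampled curves driving scale and dynamic material parameters.
# A second emits drifting sprites that shift red to blue, with spawn rate, area, speed and drag scaled by Stage.
# Both apply System.Scale to size and modulate alpha by NearFade and time of day.
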NiagaraSystem /DragonCartNarrative/Assets/VFX/Cracks/NS_DCN_Crack_Grow
dependencies {
 usedDataInterfaces [&NiagaraDataInterfaceCurve, &NiagaraDataInterfaceColorCurve]
 usedDynamicInputs [&LerpLinearColors, &FN_LOD_Distance_Remap]
 usedModules [
  &SystemState
  &self:SystemSpawnScript.SystemScriptSource.SystemScriptGraph.NiagaraNodeAssignment_0.SetVariables_#1
  &self:SystemSpawnScript.SystemScriptSource.SystemScriptGraph.NiagaraNodeAssignment_2.SetVariables_#2
 ]
}
emitters {
 NE_Red_Rift {
  emitterUpdateScript [
   &EmitterState(Life Cycle Mode=::NewEnumerator1, Loop Behavior=::NewEnumerator1, Loop Duration Mode=::NewEnumerator1, UseLoopDelay=false)
   &SpawnBurst_Instantaneous(Use Spawn Probability=false, UseLoopCountLimit=false)
   &SpawnBurst_Instantaneous(Spawn Count=2, Spawn Group=1, Use Spawn Probability=false, UseLoopCountLimit=false) {module=SpawnBurst_Instantaneous001}
   &SpawnBurst_Instantaneous(Spawn Group=2, Use Spawn Probability=false, UseLoopCountLimit=false) {module=SpawnBurst_Instantaneous002}
  ]
  particleSpawnScript [
   /SetVariables(Particles.VisibilityTag=&SpawnGroup(Max Spawn Group=2)) {isSetParameters=true}
   &InitializeParticle(
    Color=&MergeHelper_ColorSwitch(0=<rgba 2 0.5729 0.5729 1>, 1=<rgba 2 0.0833 0.1336 1>, 2=?, 3=?, Index=&SpawnGroup(), Num=::NewEnumerator0)
    Lifetime=20
    Mesh Scale=&MergeHelper_VectorSwitch(
     0=<1000 1000 1000>
     1=<0 0 0>
     2=&Multiply_VectorByFloat(Vector=&RandomRangeVector(Minimum=<1 1 1>, Maximum=<2 3 3>), Float=1000)
     3=?
     Index=&SpawnGroup()
     Num=::NewEnumerator1
    )
    Mesh Scale Mode=::NewEnumerator4
    Sprite Rotation Mode=::NewEnumerator0
    Sprite Size=<70000 25000>
    Sprite Size Mode=::NewEnumerator4
    UsePositionOffset=false
    Write Material Random=true
   )
   /SetVariables(Particles.MeshRotationEuler:Vector3f=&MergeHelper_VectorSwitch(0=<0 -0.54 0.48>, 1=<0 0 0>, 2=<0.8757 0.16 0.288>, 3=?, Index=&SpawnGroup(), Num=::NewEnumerator1)) {isSetParameters=true}
   /SetVariables(Particles.MeshOrientation=/CascadeToNiagaraConverter/NiagaraScripts/DynamicInputScripts/MakeQuatFromEuler(Euler:Vector3f=$Particles.MeshRotationEuler, Scale:NiagaraFloat=360)) {isSetParameters=true}
   &SpriteFacingAndAlignment(Alignment Coordinate Space=::Local, Facing Coordinate Space=::Local, Sprite Facing=<-0.2 0.048 1>, WriteAlignment=true, WriteFacing=true)
  ]
  particleUpdateScript [
   &ParticleState(Kill Particles When Lifetime Has Elapsed=false, Let Infinitely Lived Particles Die When Emitter Deactivates=true)
   &ScaleSpriteSize(Scale Sprite Size Mode=::NewEnumerator0, Uniform Scale Factor=&Multiply_Float(A=$System.Scale, B=$User.Stage))
   &ScaleMeshSize(Scale Factor=&VectorFromFloat(Value=$System.Scale)) {module=ScaleMeshSize001}
   &ScaleMeshSize(Scale Factor=&VectorFromFloat(Value=&FloatFromCurve(
      FloatCurve=DataInterface() {propertyValues {Curve {
         defaultValue=1
         keys [
          {arriveTangent=4.1785, interpMode=RCIM_Cubic, leaveTangent=4.1785, tangentMode=RCTM_User, time=0, value=0}
          {arriveTangent=-0.0512, interpMode=RCIM_Cubic, leaveTangent=-0.0512, tangentMode=RCTM_Break, time=0.3, value=1}
          {arriveTangent=0.0224, interpMode=RCIM_Cubic, leaveTangent=0.0224, tangentMode=RCTM_User, time=1, value=1}
         ]
        }}}
      CurveIndex=$User.Stage
     )))
   &DynamicMaterialParameters(
    Index 0 Param 1=&FloatFromCurve(
     FloatCurve=DataInterface() {propertyValues {Curve {
        defaultValue=1
        keys [
         {arriveTangent=0.1454, interpMode=RCIM_Cubic, leaveTangent=0.1454, tangentMode=RCTM_User, time=0, value=0}
         {
          arriveTangent=0.3329
          arriveTangentWeight=0.234
          interpMode=RCIM_Cubic
          leaveTangent=0.7246
          leaveTangentWeight=0.1669
          tangentMode=RCTM_Break
          tangentWeightMode=RCTWM_WeightedBoth
          time=0.7118
          value=0.1609
         }
         {arriveTangent=5.4184, tangentMode=RCTM_User, time=1, value=1}
        ]
       }}}
     CurveIndex=$User.Stage
    )
    Index 0 Param 2=&MergeHelper_FloatSwitch(
     0=&FloatFromCurve(
      FloatCurve=DataInterface() {propertyValues {Curve {
         defaultValue=1
         keys [
          {arriveTangent=1.3092, interpMode=RCIM_Cubic, leaveTangent=1.3092, tangentMode=RCTM_User, time=0, value=1}
          {arriveTangent=7, interpMode=RCIM_Cubic, leaveTangent=6.5494, tangentMode=RCTM_Break, time=0.7041, value=3.6597}
          {arriveTangent=22.234, tangentMode=RCTM_User, time=1, value=8}
         ]
        }}}
      CurveIndex=$User.Stage
     )
     1=0
     2=&FloatFromCurve(
      FloatCurve=DataInterface() {propertyValues {Curve {
         defaultValue=1
         keys [
          {arriveTangent=1.3092, interpMode=RCIM_Cubic, leaveTangent=1.3092, tangentMode=RCTM_User, time=0, value=1}
          {arriveTangent=16.7614, tangentMode=RCTM_User, time=1, value=8}
         ]
        }}}
      CurveIndex=$User.Stage
     )
     3=?
     Index=&SpawnGroup()
     Num=::NewEnumerator1
    )
    Index 0 Param 3=2
    Module.Param0WriteEnabled=true
    Module.Param1WriteEnabled=true
    Module.Param2WriteEnabled=true
    Module.Param3WriteEnabled=true
   )
   /SetVariables(Particles.FlareCameraOffset=-100, Particles.FlareDynamic:Vector4f=&MakeVector4(X=0.8, Y=1, Z=0, W=0)) {isSetParameters=true}
   &ScaleColor(
    Scale Alpha=&MergeHelper_FloatSwitch(
     0=&Multiply_Float(A=$System.NearFade, B=&FNTimeOfDayScalar(Night Scale=0.9, Morning Scale=1, Evening Scale=0.9))
     1=&Multiply_Float(A=$System.NearFade, B=&FNTimeOfDayScalar(Night Scale=0.6, Morning Scale=0.8, Evening Scale=0.8))
     2=$System.NearFade
     3=?
     Index=&SpawnGroup()
     Num=::NewEnumerator0
    )
    ScaleA=true
    ScaleRGB=true
   )
  ]
  rendererClasses [&NiagaraMeshRendererProperties, &NiagaraSpriteRendererProperties]
  renderers [
   {rendererClass=&NiagaraMeshRendererProperties, rendererIndex=0}
   {rendererClass=&NiagaraSpriteRendererProperties, rendererIndex=1}
   {rendererClass=&NiagaraMeshRendererProperties, rendererIndex=2}
  ]
 }
 NE_Red_Rift001 {
  emitterUpdateScript [
   &EmitterState(Life Cycle Mode=::NewEnumerator1, Loop Behavior=::NewEnumerator1, Loop Duration Mode=::NewEnumerator1, UseLoopDelay=false)
   &SpawnBurst_Instantaneous(Spawn Count=3, Use Spawn Probability=false, UseLoopCountLimit=false)
   &SpawnRate(SpawnGroup=1, SpawnRate=&Multiply_Float(A=5, B=$User.Stage), Use Spawn Probability=false)
  ]
  particleSpawnScript [
   &InitializeParticle(
    Color Mode=::NewEnumerator0
    Lifetime=&RandomRangeFloat(Minimum=6, Maximum=9)
    Sprite Rotation Mode=::NewEnumerator2
    Sprite Size Mode=::NewEnumerator1
    Uniform Sprite Size Max=5000
    Uniform Sprite Size Min=2000
    UsePositionOffset=false
    Write Material Random=true
   )
   &ShapeLocation(
    Box Size=&Multiply_Vector(A=<10000 2000 1000>, B=&VectorFromFloat(Value=$User.Stage))
    MaskSpawnGroup=false
    Override Seed=false
    Shape Primitive=::NewEnumerator2
    Surface Only=false
    UseOwnerScale=true
   )
   &AddVelocity(
    Module.Use Falloff Distance=false
    Override Seed=false
    Velocity Mode=::NewEnumerator1
    Velocity Speed=&Multiply_Float(A=&RandomRangeFloat(Minimum=1200, Maximum=2200), B=$User.Stage)
   )
  ]
  particleUpdateScript [
   &ParticleState()
   &ScaleSpriteSize(
    Scale Sprite Size Mode=::NewEnumerator0
    Uniform Scale Factor=&Multiply_Float(
     A=$System.Scale
     B=&FloatFromCurve(FloatCurve=DataInterface() {propertyValues {Curve {
         defaultValue=1
         keys [
          {arriveTangent=9.0725, interpMode=RCIM_Cubic, leaveTangent=9.0725, tangentMode=RCTM_User, time=0, value=0}
          {arriveTangent=-1.0958, interpMode=RCIM_Cubic, leaveTangent=-1.0958, tangentMode=RCTM_User, time=0.2407, value=0.9265}
          {arriveTangent=-1.4993, interpMode=RCIM_Cubic, leaveTangent=-1.4993, tangentMode=RCTM_User, time=1, value=0}
         ]
        }}})
    )
   )
   &ScaleColor(
    Linear Color Curve=DataInterface() {propertyValues {
      AlphaCurve {keys [<at 0 1>,<at 1 1>]}
      BlueCurve {keys [<at 0.5418 0.0666>,<at 0.7326 2>]}
      GreenCurve {keys [<at 0.5418 0>,<at 0.7326 0.6647>]}
      RedCurve {keys [<at 0.5418 4>,<at 0.7326 0>]}
     }}
    Scale Mode=2
   ) {module=ScaleColor001}
   &ScaleColor(
    Scale Alpha=&MergeHelper_FloatSwitch(
     0=&Multiply_Float(
      A=$System.NearFade
      B=&FNTimeOfDayScalar(Night Scale=1, Morning Scale=1.2, Day Scale=1.5, Evening Scale=1)
     )
     1=&Multiply_Float(A=$System.NearFade, B=&FNTimeOfDayScalar(Night Scale=0.6, Morning Scale=0.8, Evening Scale=0.7))
     2=$System.NearFade
     3=?
     Index=&SpawnGroup()
     Num=::NewEnumerator0
    )
    ScaleA=true
    ScaleRGB=true
   )
   /SetVariables(Particles.FlareCameraOffset=100, Particles.FlareDynamic:Vector4f=&MakeVector4(X=1, Y=1, Z=0, W=0)) {isSetParameters=true}
   &Drag(
    Drag=&Multiply_Float(A=&RandomRangeFloat(Minimum=0.2, Maximum=0.7), B=$User.Stage)
    Use Linear Drag=true
    Use Rotational Drag=false
   )
   &SolveForcesAndVelocity(Clamp Velocity=false, Limit Acceleration=false)
  ]
  rendererClasses [&NiagaraSpriteRendererProperties]
  renderers [{rendererClass=&NiagaraSpriteRendererProperties}]
 }
}
summary {
 systemName=NS_DCN_Crack_Grow
 emitters [
  {emitterName=NE_Red_Rift, rendererClasses [&NiagaraMeshRendererProperties, &NiagaraSpriteRendererProperties]}
  {emitterName=NE_Red_Rift001, rendererClasses [&NiagaraSpriteRendererProperties]}
 ]
 userVariables [User.Stage {type=NiagaraFloat, defaultValue=1, description=""}]
}
