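# Niagara burst effect for a growing crack, judging by its name, built from huge red explosion layers, though every emitter in the visible portion is disabled.
# Corona, RoundFlare, Shockwave and Sphere animate size, color and dynamic material parameters through curves.
# Frags1-3 throw spinning, drag-slowed meshes, while two GPU Sparklies emitters spray flickering sprites.
# Minimal001 merges the four flash and wave layers into one emitter using spawn-group switches.
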NiagaraSystem /DragonCartNarrative/Assets/VFX/Cracks/NS_DCN_Crack_Grow_Burst
dependencies {
 usedDataInterfaces [&NiagaraDataInterfaceCurve, &NiagaraDataInterfaceVectorCurve]
 usedDynamicInputs []
 usedModules [
  &SystemState
  &self:SystemSpawnScript.SystemScriptSource.SystemScriptGraph.NiagaraNodeAssignment_0.SetVariables_#1
  &self:SystemSpawnScript.SystemScriptSource.SystemScriptGraph.NiagaraNodeAssignment_2.SetVariables_#2
 ]
}
emitters {
 Corona {
  bEnabled=false
  emitterUpdateScript [
   &EmitterState(Life Cycle Mode=::NewEnumerator1, Loop Behavior=::NewEnumerator1, Loop Duration=1, UseLoopDelay=true)
   &SpawnBurst_Instantaneous(Use Spawn Probability=false, UseLoopCountLimit=false)
  ]
  particleSpawnScript [
   &InitializeParticle(Lifetime=2, Sprite Size=<45000 45000>, Sprite Size Mode=::NewEnumerator4, UsePositionOffset=false, Write Material Random=true)
   &Color(Color=&MakeLinearColorFromVectorAndFloat("Vector (RGB)"=<1 0.2 0.2133>, "Float (Alpha)"=1))
   /CascadeToNiagaraConverter/NiagaraScripts/ModuleScripts/SpriteLockAxis(Lock Axis Type:ENiagara_SpriteLockAxis=::NewEnumerator5, UseAlignmentCoordinateSpace=false, UseFacingCoordinateSpace=false)
  ]
  particleUpdateScript [
   &ParticleState()
   &ScaleSpriteSize(
    Scale Sprite Size Mode=::NewEnumerator0
    Uniform Scale Factor=&FloatFromCurve(FloatCurve=DataInterface() {propertyValues {Curve {keys [
         {arriveTangent=1.4067, interpMode=RCIM_Cubic, leaveTangent=1.4067, tangentMode=RCTM_User, time=0, value=0.1}
         {arriveTangent=0.2191, interpMode=RCIM_Cubic, leaveTangent=0.2191, tangentMode=RCTM_User, time=1, value=1}
        ]}}})
   ) {module=ScaleSpriteSize001}
   &ScaleSpriteSize(Scale Factor=<5 5>, Scale Sprite Size Mode=::NewEnumerator2)
   &ScaleColor(
    Scale Alpha=&FloatFromCurve(FloatCurve=DataInterface() {propertyValues {Curve {keys [<at 0 0 RCIM_Cubic>,<at 0.25 1 RCIM_Cubic>,<at 1 0>]}}})
    Scale RGB=<2 2 2>
    ScaleA=true
    ScaleRGB=true
   )
   &DynamicMaterialParameters(
    Index 0 Param 1=&FloatFromCurve(FloatCurve=DataInterface() {propertyValues {Curve {keys [{arriveTangent=-2.4947, interpMode=RCIM_Cubic, leaveTangent=-2.4947, tangentMode=RCTM_User, time=0, value=1.5}, <at 1 0.2 RCIM_Cubic>]}}})
    Index 0 Param 2=0.4
    Index 0 Param 3=0.6
    Index 0 Param 4=1
    Module.Param0WriteEnabled=true
    Module.Param1WriteEnabled=true
    Module.Param2WriteEnabled=true
    Module.Param3WriteEnabled=true
   )
  ]
  rendererClasses [&NiagaraSpriteRendererProperties]
  renderers [{rendererClass=&NiagaraSpriteRendererProperties}]
 }
 Frags1 {
  bEnabled=false
  emitterUpdateScript [
   &EmitterState(Life Cycle Mode=::NewEnumerator1, Loop Behavior=::NewEnumerator1, Loop Duration=1, UseLoopDelay=true)
   &SpawnBurst_Instantaneous(Spawn Count=12, Use Spawn Probability=false, UseLoopCountLimit=false)
  ]
  particleSpawnScript [
   /SetVariables(Particles.ShapeLocationOffset:Vector3f=&RandomRangeVector(Minimum=<-1000 0 0>, Maximum=<-10000 0 0>, Evaluation Type=::NewEnumerator1)) {isSetParameters=true}
   /SetVariables(Particles.ShapeLocationOffsetSpace:ENiagaraCoordinateSpace=::Local) {isSetParameters=true}
   /SetVariables(Particles.MeshRotationEuler:Vector3f=&RandomRangeVector(Minimum=<0 0 0>, Maximum=<1 1 1>, Evaluation Type=::NewEnumerator1)) {isSetParameters=true}
   /SetVariables(Particles.MeshOrientation=/CascadeToNiagaraConverter/NiagaraScripts/DynamicInputScripts/MakeQuatFromEuler(Euler:Vector3f=$Particles.MeshRotationEuler, Scale:NiagaraFloat=360)) {isSetParameters=true}
   /SetVariables(Particles.MeshRotationRateBase:Vector3f=&RandomRangeVector(Minimum=<-2 -0.4 -0.1>, Maximum=<2 0.4 0.1>, Evaluation Type=::NewEnumerator1)) {isSetParameters=true}
   &InitializeParticle(
    Lifetime=&RandomRangeFloat(Minimum=2, Maximum=3.2, Evaluation Type=::NewEnumerator1)
    Mesh Scale=&RandomRangeVector(Minimum=<100 50 30>, Maximum=<160 70 50>, Evaluation Type=::NewEnumerator1)
    Mesh Scale Mode=::NewEnumerator4
    UsePositionOffset=false
    Write Material Random=true
   )
   &AddVelocity(Velocity=&RandomRangeVector(Minimum=<4000 0 20000>, Maximum=<160000 0 30000>, Evaluation Type=::NewEnumerator1))
   &Color(Color=&MakeLinearColorFromVectorAndFloat(
     "Vector (RGB)"=<5 5 5>
     "Float (Alpha)"=&RandomRangeFloat(Minimum=0.2, Maximum=0.6, Evaluation Type=::NewEnumerator1)
    ))
   &ShapeLocation(
    Hemisphere Distribution=<0.5 0.5>
    MaskSpawnGroup=false
    Offset=$Particles.ShapeLocationOffset
    Offset Coordinate Space=$Particles.ShapeLocationOffsetSpace
    Offset Mode=::NewEnumerator0
    Override Seed=false
    Rotation Mode=::NewEnumerator2
    Sphere Radius=&RandomRangeFloat(Minimum=200, Maximum=400, Evaluation Type=::NewEnumerator1)
    UseOwnerScale=true
    Yaw / Pitch / Roll=<0.75 0 0>
   )
   /CascadeToNiagaraConverter/NiagaraScripts/ModuleScripts/FixShapeVector(
    Offset:Vector3f=$Particles.ShapeLocationOffset
    Offset Coordinate Space:ENiagaraCoordinateSpace=$Particles.ShapeLocationOffsetSpace
    Offset Mode:ENiagara_OffsetMode=::NewEnumerator0
   )
   &AddVelocity(Velocity=&Multiply_VectorByFloat(
     Vector=$Output.FixShapeVector.ShapeVector
     Float=&RandomRangeFloat(Minimum=50, Maximum=180, Evaluation Type=::NewEnumerator1)
    )) {module=AddVelocity001}
  ]
  particleUpdateScript [
   /SetVariables(Particles.MeshRotationRate:Vector3f=$Particles.MeshRotationRateBase) {isSetParameters=true}
   /SetVariables(Particles.MeshRotationRate:Vector3f=&Multiply_Vector(
     A=$Particles.MeshRotationRate
     B=&VectorFromFloat(Value=&FloatFromCurve(FloatCurve=DataInterface() {propertyValues {Curve {keys [
           {arriveTangent=-1.4943, interpMode=RCIM_Cubic, leaveTangent=-1.4943, tangentMode=RCTM_User, time=0, value=1}
           {arriveTangent=-0.2218, interpMode=RCIM_Cubic, leaveTangent=-0.2218, tangentMode=RCTM_User, time=1, value=0}
          ]}}}))
    )) {isSetParameters=true}
   &ParticleState()
   &ScaleMeshSize(Scale Factor=&VectorFromCurve(VectorCurve=DataInterface() {propertyValues {
       XCurve {keys [<at 0.7 1 RCIM_Cubic>,<at 1 0>]}
       YCurve {keys [<at 0.7 1 RCIM_Cubic>,<at 1 1>]}
       ZCurve {keys [<at 0.7 1 RCIM_Cubic>,<at 1 1>]}
      }}))
   &Drag(
    Drag=&RandomRangeFloat(Minimum=0.75, Maximum=1.5, Evaluation Type=::NewEnumerator1)
    Use Linear Drag=true
    Use Rotational Drag=true
   )
   &SolveForcesAndVelocity(Clamp Velocity=false, Limit Acceleration=false)
   &ScaleColor(
    Scale Alpha=&FloatFromCurve(FloatCurve=DataInterface() {propertyValues {Curve {keys [<at 0.7 1 RCIM_Cubic>,<at 1 0>]}}})
    Scale RGB=&VectorFromFloat(Value=&FloatFromCurve(FloatCurve=DataInterface() {propertyValues {Curve {keys [<at 0.1 8 RCIM_Cubic>,<at 0.7 1>]}}}))
    ScaleA=true
    ScaleRGB=true
   )
   /CascadeToNiagaraConverter/NiagaraScripts/ModuleScripts/UpdateMeshOrientationByEuler(RotationRate:Vector3f=$Particles.MeshRotationRate)
  ]
  rendererClasses [&NiagaraMeshRendererProperties]
  renderers [{rendererClass=&NiagaraMeshRendererProperties, rendererIndex=0}]
 }
 Frags2 {
  bEnabled=false
  emitterUpdateScript [
   &EmitterState(Life Cycle Mode=::NewEnumerator1, Loop Behavior=::NewEnumerator1, Loop Duration=1, UseLoopDelay=true)
   &SpawnBurst_Instantaneous(Spawn Count=12, Use Spawn Probability=false, UseLoopCountLimit=false)
  ]
  particleSpawnScript [
   /SetVariables(Particles.ShapeLocationOffset:Vector3f=&RandomRangeVector(Minimum=<-1000 0 0>, Maximum=<-10000 0 0>, Evaluation Type=::NewEnumerator1)) {isSetParameters=true}
   /SetVariables(Particles.ShapeLocationOffsetSpace:ENiagaraCoordinateSpace=::Local) {isSetParameters=true}
   /SetVariables(Particles.MeshRotationEuler:Vector3f=&RandomRangeVector(Minimum=<0 0 0>, Maximum=<1 1 1>, Evaluation Type=::NewEnumerator1)) {isSetParameters=true}
   /SetVariables(Particles.MeshOrientation=&MakeQuatFromEuler(Euler:Vector3f=$Particles.MeshRotationEuler, Scale:NiagaraFloat=360)) {isSetParameters=true}
   /SetVariables(Particles.MeshRotationRateBase:Vector3f=&RandomRangeVector(Minimum=<-2 -0.4 -0.1>, Maximum=<2 0.4 0.1>, Evaluation Type=::NewEnumerator1)) {isSetParameters=true}
   &InitializeParticle(
    Lifetime=&RandomRangeFloat(Minimum=2, Maximum=3.2, Evaluation Type=::NewEnumerator1)
    Mesh Scale=&RandomRangeVector(Minimum=<100 50 30>, Maximum=<160 70 50>, Evaluation Type=::NewEnumerator1)
    Mesh Scale Mode=::NewEnumerator4
    UsePositionOffset=false
    Write Material Random=true
   )
   &AddVelocity(Velocity=&RandomRangeVector(Minimum=<4000 0 20000>, Maximum=<160000 0 30000>, Evaluation Type=::NewEnumerator1))
   &Color(Color=&MakeLinearColorFromVectorAndFloat(
     "Vector (RGB)"=<5 5 5>
     "Float (Alpha)"=&RandomRangeFloat(Minimum=0.2, Maximum=0.6, Evaluation Type=::NewEnumerator1)
    ))
   &ShapeLocation(
    Hemisphere Distribution=<0.5 0.5>
    MaskSpawnGroup=false
    Offset=$Particles.ShapeLocationOffset
    Offset Coordinate Space=$Particles.ShapeLocationOffsetSpace
    Offset Mode=::NewEnumerator0
    Override Seed=false
    Rotation Mode=::NewEnumerator2
    Sphere Radius=&RandomRangeFloat(Minimum=200, Maximum=400, Evaluation Type=::NewEnumerator1)
    UseOwnerScale=true
    Yaw / Pitch / Roll=<0.75 0 0>
   )
   &FixShapeVector(
    Offset:Vector3f=$Particles.ShapeLocationOffset
    Offset Coordinate Space:ENiagaraCoordinateSpace=$Particles.ShapeLocationOffsetSpace
    Offset Mode:ENiagara_OffsetMode=::NewEnumerator0
   )
   &AddVelocity(Velocity=&Multiply_VectorByFloat(
     Vector=$Output.FixShapeVector.ShapeVector
     Float=&RandomRangeFloat(Minimum=50, Maximum=180, Evaluation Type=::NewEnumerator1)
    )) {module=AddVelocity001}
  ]
  particleUpdateScript [
   /SetVariables(Particles.MeshRotationRate:Vector3f=$Particles.MeshRotationRateBase) {isSetParameters=true}
   /SetVariables(Particles.MeshRotationRate:Vector3f=&Multiply_Vector(
     A=$Particles.MeshRotationRate
     B=&VectorFromFloat(Value=&FloatFromCurve(FloatCurve=DataInterface() {propertyValues {Curve {keys [
           {arriveTangent=-1.4943, interpMode=RCIM_Cubic, leaveTangent=-1.4943, tangentMode=RCTM_User, time=0, value=1}
           {arriveTangent=-0.2218, interpMode=RCIM_Cubic, leaveTangent=-0.2218, tangentMode=RCTM_User, time=1, value=0}
          ]}}}))
    )) {isSetParameters=true}
   &ParticleState()
   &ScaleMeshSize(Scale Factor=&VectorFromCurve(VectorCurve=DataInterface() {propertyValues {
       XCurve {keys [<at 0.7 1 RCIM_Cubic>,<at 1 0>]}
       YCurve {keys [<at 0.7 1 RCIM_Cubic>,<at 1 1>]}
       ZCurve {keys [<at 0.7 1 RCIM_Cubic>,<at 1 1>]}
      }}))
   &Drag(
    Drag=&RandomRangeFloat(Minimum=0.75, Maximum=1.5, Evaluation Type=::NewEnumerator1)
    Use Linear Drag=true
    Use Rotational Drag=true
   )
   &SolveForcesAndVelocity(Clamp Velocity=false, Limit Acceleration=false)
   &ScaleColor(
    Scale Alpha=&FloatFromCurve(FloatCurve=DataInterface() {propertyValues {Curve {keys [<at 0.7 1 RCIM_Cubic>,<at 1 0>]}}})
    Scale RGB=&VectorFromFloat(Value=&FloatFromCurve(FloatCurve=DataInterface() {propertyValues {Curve {keys [<at 0.1 8 RCIM_Cubic>,<at 0.7 1>]}}}))
    ScaleA=true
    ScaleRGB=true
   )
   &UpdateMeshOrientationByEuler(RotationRate:Vector3f=$Particles.MeshRotationRate)
  ]
  rendererClasses [&NiagaraMeshRendererProperties]
  renderers [{rendererClass=&NiagaraMeshRendererProperties, rendererIndex=0}]
 }
 Frags3 {
  bEnabled=false
  emitterUpdateScript [
   &EmitterState(Life Cycle Mode=::NewEnumerator1, Loop Behavior=::NewEnumerator1, Loop Duration=1, UseLoopDelay=true)
   &SpawnBurst_Instantaneous(Spawn Count=12, Use Spawn Probability=false, UseLoopCountLimit=false)
  ]
  particleSpawnScript [
   /SetVariables(Particles.ShapeLocationOffset:Vector3f=&RandomRangeVector(Minimum=<-1000 0 0>, Maximum=<-10000 0 0>, Evaluation Type=::NewEnumerator1)) {isSetParameters=true}
   /SetVariables(Particles.ShapeLocationOffsetSpace:ENiagaraCoordinateSpace=::Local) {isSetParameters=true}
   /SetVariables(Particles.MeshRotationEuler:Vector3f=&RandomRangeVector(Minimum=<0 0 0>, Maximum=<1 1 1>, Evaluation Type=::NewEnumerator1)) {isSetParameters=true}
   /SetVariables(Particles.MeshOrientation=&MakeQuatFromEuler(Euler:Vector3f=$Particles.MeshRotationEuler, Scale:NiagaraFloat=360)) {isSetParameters=true}
   /SetVariables(Particles.MeshRotationRateBase:Vector3f=&RandomRangeVector(Minimum=<-2 -0.4 -0.1>, Maximum=<2 0.4 0.1>, Evaluation Type=::NewEnumerator1)) {isSetParameters=true}
   &InitializeParticle(
    Lifetime=&RandomRangeFloat(Minimum=2, Maximum=3.2, Evaluation Type=::NewEnumerator1)
    Mesh Scale=&RandomRangeVector(Minimum=<100 50 30>, Maximum=<160 70 50>, Evaluation Type=::NewEnumerator1)
    Mesh Scale Mode=::NewEnumerator4
    UsePositionOffset=false
    Write Material Random=true
   )
   &AddVelocity(Velocity=&RandomRangeVector(Minimum=<4000 0 20000>, Maximum=<160000 0 30000>, Evaluation Type=::NewEnumerator1))
   &Color(Color=&MakeLinearColorFromVectorAndFloat(
     "Vector (RGB)"=<5 5 5>
     "Float (Alpha)"=&RandomRangeFloat(Minimum=0.2, Maximum=0.6, Evaluation Type=::NewEnumerator1)
    ))
   &ShapeLocation(
    Hemisphere Distribution=<0.5 0.5>
    MaskSpawnGroup=false
    Offset=$Particles.ShapeLocationOffset
    Offset Coordinate Space=$Particles.ShapeLocationOffsetSpace
    Offset Mode=::NewEnumerator0
    Override Seed=false
    Rotation Mode=::NewEnumerator2
    Sphere Radius=&RandomRangeFloat(Minimum=200, Maximum=400, Evaluation Type=::NewEnumerator1)
    UseOwnerScale=true
    Yaw / Pitch / Roll=<0.75 0 0>
   )
   &FixShapeVector(
    Offset:Vector3f=$Particles.ShapeLocationOffset
    Offset Coordinate Space:ENiagaraCoordinateSpace=$Particles.ShapeLocationOffsetSpace
    Offset Mode:ENiagara_OffsetMode=::NewEnumerator0
   )
   &AddVelocity(Velocity=&Multiply_VectorByFloat(
     Vector=$Output.FixShapeVector.ShapeVector
     Float=&RandomRangeFloat(Minimum=50, Maximum=180, Evaluation Type=::NewEnumerator1)
    )) {module=AddVelocity001}
  ]
  particleUpdateScript [
   /SetVariables(Particles.MeshRotationRate:Vector3f=$Particles.MeshRotationRateBase) {isSetParameters=true}
   /SetVariables(Particles.MeshRotationRate:Vector3f=&Multiply_Vector(
     A=$Particles.MeshRotationRate
     B=&VectorFromFloat(Value=&FloatFromCurve(FloatCurve=DataInterface() {propertyValues {Curve {keys [
           {arriveTangent=-1.4943, interpMode=RCIM_Cubic, leaveTangent=-1.4943, tangentMode=RCTM_User, time=0, value=1}
           {arriveTangent=-0.2218, interpMode=RCIM_Cubic, leaveTangent=-0.2218, tangentMode=RCTM_User, time=1, value=0}
          ]}}}))
    )) {isSetParameters=true}
   &ParticleState()
   &ScaleMeshSize(Scale Factor=&VectorFromCurve(VectorCurve=DataInterface() {propertyValues {
       XCurve {keys [<at 0.7 1 RCIM_Cubic>,<at 1 0>]}
       YCurve {keys [<at 0.7 1 RCIM_Cubic>,<at 1 1>]}
       ZCurve {keys [<at 0.7 1 RCIM_Cubic>,<at 1 1>]}
      }}))
   &Drag(
    Drag=&RandomRangeFloat(Minimum=0.75, Maximum=1.5, Evaluation Type=::NewEnumerator1)
    Use Linear Drag=true
    Use Rotational Drag=true
   )
   &SolveForcesAndVelocity(Clamp Velocity=false, Limit Acceleration=false)
   &ScaleColor(
    Scale Alpha=&FloatFromCurve(FloatCurve=DataInterface() {propertyValues {Curve {keys [<at 0.7 1 RCIM_Cubic>,<at 1 0>]}}})
    Scale RGB=&VectorFromFloat(Value=&FloatFromCurve(FloatCurve=DataInterface() {propertyValues {Curve {keys [<at 0.1 8 RCIM_Cubic>,<at 0.7 1>]}}}))
    ScaleA=true
    ScaleRGB=true
   )
   &UpdateMeshOrientationByEuler(RotationRate:Vector3f=$Particles.MeshRotationRate)
  ]
  rendererClasses [&NiagaraMeshRendererProperties]
  renderers [{rendererClass=&NiagaraMeshRendererProperties, rendererIndex=0}]
 }
 Minimal001 {
  bEnabled=false
  emitterUpdateScript [
   &EmitterState(Life Cycle Mode=::NewEnumerator1, Loop Behavior=::NewEnumerator1, Loop Duration=1, UseLoopDelay=false)
   &SpawnBurst_Instantaneous(Use Spawn Probability=false, UseLoopCountLimit=false)
   &SpawnBurst_Instantaneous(Spawn Group=1, Use Spawn Probability=false, UseLoopCountLimit=false) {module=SpawnBurst_Instantaneous001}
   &SpawnBurst_Instantaneous(Spawn Group=2, Use Spawn Probability=false, UseLoopCountLimit=false) {module=SpawnBurst_Instantaneous002}
   &SpawnBurst_Instantaneous(Spawn Group=3, Use Spawn Probability=false, UseLoopCountLimit=false) {module=SpawnBurst_Instantaneous003}
  ]
  particleSpawnScript [&InitializeParticle(
    Color=&MergeHelper_ColorSwitch(
     0=<rgba 1 0.2 0.2133 1>
     1=<rgba 10 2 2 0.025>
     2=<rgba 10 0.25 0.25 1>
     3=<rgba 3 0.075 0.075 1>
     Index=&SpawnGroup()
     Num=::NewEnumerator2
    )
    Lifetime=&MergeHelper_FloatSwitch(0=2, 1=2, 2=6, 3=0.5, Index=&SpawnGroup(), Num=::NewEnumerator2)
    Mesh Scale=&MergeHelper_VectorSwitch(0=<0 0 0>, 1=<700 700 700>, 2=<700 700 100>, 3=<0 0 0>, Index=&SpawnGroup(), Num=::NewEnumerator2)
    Mesh Scale Mode=::NewEnumerator4
    Sprite Size=&MergeHelper_Vector2Switch(0=<45000 45000>, 1=<0 0>, 2=<0 0>, 3=<160000 160000>, 4=?, Index=&SpawnGroup(), Num=::NewEnumerator2)
    Sprite Size Mode=::NewEnumerator4
    UsePositionOffset=false
    Write Material Random=true
   )]
  particleUpdateScript [
   &ParticleState()
   &ScaleSpriteSize(
    Scale Sprite Size Mode=::NewEnumerator0
    Uniform Scale Factor=&MergeHelper_FloatSwitch(
     0=&FloatFromCurve(
      FloatCurve=DataInterface() {propertyValues {Curve {keys [
          {arriveTangent=1.4067, interpMode=RCIM_Cubic, leaveTangent=1.4067, tangentMode=RCTM_User, time=0, value=0.1}
          {arriveTangent=0.2191, interpMode=RCIM_Cubic, leaveTangent=0.2191, tangentMode=RCTM_User, time=1, value=1}
         ]}}}
      Scale Curve=5
     )
     1=0
     2=0
     3=&FloatFromCurve(FloatCurve=DataInterface() {propertyValues {Curve {keys [
          {arriveTangent=5.3033, interpMode=RCIM_Cubic, leaveTangent=5.3033, tangentMode=RCTM_User, time=0, value=1}
          {arriveTangent=-2.1603, interpMode=RCIM_Cubic, leaveTangent=-2.1603, tangentMode=RCTM_User, time=0.1595, value=1.2315}
          {arriveTangent=-0.3386, interpMode=RCIM_Cubic, leaveTangent=-0.3386, tangentMode=RCTM_User, time=1, value=0.2}
         ]}}})
     Index=&SpawnGroup()
     Num=::NewEnumerator2
    )
   )
   &ScaleMeshSize(Scale Factor=&MergeHelper_VectorSwitch(
     0=<0 0 0>
     1=&VectorFromFloat(Value=&FloatFromCurve(
       FloatCurve=DataInterface() {propertyValues {Curve {keys [
           {arriveTangent=1.5009, interpMode=RCIM_Cubic, leaveTangent=1.5009, tangentMode=RCTM_User, time=0, value=0.1762}
           {arriveTangent=0.8097, interpMode=RCIM_Cubic, leaveTangent=0.8097, time=0.4876, value=0.6963}
           {arriveTangent=0.1287, interpMode=RCIM_Cubic, leaveTangent=0.1287, tangentMode=RCTM_User, time=1, value=0.9859}
          ]}}}
       Scale Curve=5
      ))
     2=&VectorFromFloat(Value=&FloatFromCurve(FloatCurve=DataInterface() {propertyValues {Curve {keys [
           {arriveTangent=1.3724, interpMode=RCIM_Cubic, leaveTangent=1.3724, tangentMode=RCTM_User, time=0, value=0.1762}
           {arriveTangent=0.8097, interpMode=RCIM_Cubic, leaveTangent=0.8097, time=0.4671, value=0.7411}
           {arriveTangent=0.1287, interpMode=RCIM_Cubic, leaveTangent=0.1287, tangentMode=RCTM_User, time=1, value=0.9859}
          ]}}}))
     3=<0 0 0>
     Index=&SpawnGroup()
     Num=::NewEnumerator2
    ))
   &ScaleColor(
    Scale Alpha=&MergeHelper_FloatSwitch(
     0=&FloatFromCurve(FloatCurve=DataInterface() {propertyValues {Curve {keys [<at 0 0 RCIM_Cubic>,<at 0.25 1 RCIM_Cubic>,<at 1 0>]}}})
     1=&FloatFromCurve(FloatCurve=DataInterface() {propertyValues {Curve {keys [
          {arriveTangent=-0.2749, interpMode=RCIM_Cubic, leaveTangent=-0.2749, tangentMode=RCTM_User, time=0, value=1}
          {arriveTangent=-0.9418, leaveTangent=-0.9418, time=1, value=0}
         ]}}})
     2=&FloatFromCurve(FloatCurve=DataInterface() {propertyValues {Curve {keys [<at 0 1 RCIM_Cubic>, {arriveTangent=-1.8559, leaveTangent=-1.8559, time=1, value=0}]}}})
     3=&FloatFromCurve(FloatCurve=DataInterface() {propertyValues {Curve {keys [<at 0 0.7 RCIM_Cubic>, {arriveTangent=-1.3318, interpMode=RCIM_Cubic, leaveTangent=-1.3318, tangentMode=RCTM_User, time=1, value=0}]}}})
     Index=&SpawnGroup()
     Num=::NewEnumerator2
    )
    Scale RGB=&MergeHelper_VectorSwitch(
     0=<2 2 2>
     1=&VectorFromCurve(VectorCurve=DataInterface() {propertyValues {XCurve {keys [<at 0 1>,<at 1 1>]}, YCurve {keys [<at 0 1>,<at 1 1>]}, ZCurve {keys [<at 0 1>,<at 1 1>]}}})
     2=&VectorFromCurve(VectorCurve=DataInterface() {propertyValues {XCurve {keys [<at 0 1>,<at 1 1>]}, YCurve {keys [<at 0 1>,<at 1 1>]}, ZCurve {keys [<at 0 1>,<at 1 1>]}}})
     3=&VectorFromFloat(Value=&FloatFromCurve(FloatCurve=DataInterface() {propertyValues {Curve {keys [<at 0 10 RCIM_Cubic>,<at 0.05 1>]}}}))
     Index=&SpawnGroup()
     Num=::NewEnumerator2
    )
    ScaleA=true
    ScaleRGB=true
   )
   /SetVariables(
    Particles.FlareCoronaDyn:Vector4f=&MakeVector4(
     X=&FloatFromCurve(FloatCurve=DataInterface() {propertyValues {Curve {keys [{arriveTangent=-2.4947, interpMode=RCIM_Cubic, leaveTangent=-2.4947, tangentMode=RCTM_User, time=0, value=1.5}, <at 1 0.2 RCIM_Cubic>]}}})
     Y=0.4
     Z=0.6
     W=1
    )
    Particles.RoundFlare:Vector4f=&MakeVector4(
     X=&FloatFromCurve(FloatCurve=DataInterface() {propertyValues {Curve {keys [<at 0 1 RCIM_Cubic>,<at 0.8 0>]}}})
     Y=2
     Z=1
     W=1
    )
    Particles.ShockwaveRingDyn:Vector4f=&MakeVector4(
     X=&FloatFromCurve(FloatCurve=DataInterface() {propertyValues {Curve {keys [
          {arriveTangent=-1.3562, interpMode=RCIM_Cubic, leaveTangent=-1.3562, tangentMode=RCTM_User, time=0, value=2}
          {arriveTangent=-3.7058, interpMode=RCIM_Cubic, leaveTangent=-3.7058, tangentMode=RCTM_User, time=1, value=-0.5}
         ]}}})
     Y=0.3
     Z=3
     W=1
    )
    Particles.SphereWaveDyn:Vector4f=&MakeVector4(
     X=&FloatFromCurve(FloatCurve=DataInterface() {propertyValues {Curve {keys [<at 0.5 0.95>,<at 1 1>]}}})
     Y=&FloatFromCurve(FloatCurve=DataInterface() {propertyValues {Curve {keys [
          <at 0 1 RCIM_Cubic>
          {arriveTangent=-2, interpMode=RCIM_Cubic, leaveTangent=-2, time=0.5358, value=0.6834}
          {arriveTangent=-1.5094, interpMode=RCIM_Cubic, leaveTangent=-1.5094, tangentMode=RCTM_User, time=1, value=-1}
         ]}}})
     Z=1
     W=1
    )
   ) {isSetParameters=true}
  ]
  rendererClasses [&NiagaraSpriteRendererProperties, &NiagaraMeshRendererProperties]
  renderers [
   {rendererClass=&NiagaraSpriteRendererProperties}
   {rendererClass=&NiagaraMeshRendererProperties, rendererIndex=1}
   {rendererClass=&NiagaraMeshRendererProperties, rendererIndex=2}
   {rendererClass=&NiagaraSpriteRendererProperties, rendererIndex=3}
  ]
 }
 Minimal002 {
  bEnabled=false
  emitterUpdateScript [&EmitterState()]
  particleSpawnScript [&InitializeParticle(Color Mode=::NewEnumerator0, Lifetime=5, UsePositionOffset=false, Write Material Random=true)]
  particleUpdateScript [&ParticleState()]
  rendererClasses [&NiagaraSpriteRendererProperties]
  renderers [{rendererClass=&NiagaraSpriteRendererProperties}]
 }
 RoundFlare {
  bEnabled=false
  emitterUpdateScript [
   &EmitterState(Inactive Response=::NewEnumerator1, Life Cycle Mode=::NewEnumerator1, Loop Behavior=::NewEnumerator1, Loop Duration=1, UseLoopDelay=true)
   &SpawnBurst_Instantaneous(Use Spawn Probability=false, UseLoopCountLimit=false)
  ]
  particleSpawnScript [
   &InitializeParticle(Lifetime=0.5, Sprite Size=<160000 160000>, Sprite Size Mode=::NewEnumerator4, UsePositionOffset=false, Write Material Random=true)
   &Color(Color=&MakeLinearColorFromVectorAndFloat("Vector (RGB)"=<3 0.075 0.075>, "Float (Alpha)"=1))
   /CascadeToNiagaraConverter/NiagaraScripts/ModuleScripts/AddVectorToPosition(Vector:Vector3f=&TransformVector(Vector=<-4000 0 0>, Destination Space=::Simulation))
  ]
  particleUpdateScript [
   &ParticleState()
   &ScaleSpriteSize(
    Scale Sprite Size Mode=::NewEnumerator0
    Uniform Scale Factor=&FloatFromCurve(FloatCurve=DataInterface() {propertyValues {Curve {keys [
         {arriveTangent=5.3033, interpMode=RCIM_Cubic, leaveTangent=5.3033, tangentMode=RCTM_User, time=0, value=1}
         {arriveTangent=-2.1603, interpMode=RCIM_Cubic, leaveTangent=-2.1603, tangentMode=RCTM_User, time=0.1595, value=1.2315}
         {arriveTangent=-0.3386, interpMode=RCIM_Cubic, leaveTangent=-0.3386, tangentMode=RCTM_User, time=1, value=0.2}
        ]}}})
   )
   &ScaleColor(
    Scale Alpha=&FloatFromCurve(FloatCurve=DataInterface() {propertyValues {Curve {keys [<at 0 0.7 RCIM_Cubic>, {arriveTangent=-1.3318, interpMode=RCIM_Cubic, leaveTangent=-1.3318, tangentMode=RCTM_User, time=1, value=0}]}}})
    Scale RGB=&VectorFromFloat(Value=&FloatFromCurve(FloatCurve=DataInterface() {propertyValues {Curve {keys [<at 0 10 RCIM_Cubic>,<at 0.05 1>]}}}))
    ScaleA=true
    ScaleRGB=true
   )
   &DynamicMaterialParameters(
    Index 0 Param 1=&FloatFromCurve(FloatCurve=DataInterface() {propertyValues {Curve {keys [<at 0 1 RCIM_Cubic>,<at 0.8 0>]}}})
    Index 0 Param 2=2
    Index 0 Param 3=1
    Index 0 Param 4=1
    Module.Param0WriteEnabled=true
    Module.Param1WriteEnabled=true
    Module.Param2WriteEnabled=true
    Module.Param3WriteEnabled=true
   )
   &CameraOffset(Camera Offset Amount=800)
  ]
  rendererClasses [&NiagaraSpriteRendererProperties]
  renderers [{rendererClass=&NiagaraSpriteRendererProperties}]
 }
 Shockwave {
  bEnabled=false
  emitterUpdateScript [
   &EmitterState(Life Cycle Mode=::NewEnumerator1, Loop Behavior=::NewEnumerator1, Loop Duration=1, UseLoopDelay=true)
   &SpawnBurst_Instantaneous(Use Spawn Probability=false, UseLoopCountLimit=false)
  ]
  particleSpawnScript [
   /SetVariables() {isSetParameters=true}
   /SetVariables(Particles.MeshOrientation=&MakeQuatFromEuler(Euler:Vector3f=$Particles.MeshRotationEuler, Scale:NiagaraFloat=360)) {isSetParameters=true}
   &InitializeParticle(Lifetime=6, Mesh Scale=<700 700 100>, Mesh Scale Mode=::NewEnumerator4, UsePositionOffset=false, Write Material Random=true)
   &Color(Color=&MakeLinearColorFromVectorAndFloat("Vector (RGB)"=<10 0.25 0.25>, "Float (Alpha)"=1))
  ]
  particleUpdateScript [
   &ParticleState()
   &ScaleMeshSize(Scale Factor=&VectorFromFloat(Value=&FloatFromCurve(FloatCurve=DataInterface() {propertyValues {Curve {keys [
          {arriveTangent=1.3724, interpMode=RCIM_Cubic, leaveTangent=1.3724, tangentMode=RCTM_User, time=0, value=0.1762}
          {arriveTangent=0.8097, interpMode=RCIM_Cubic, leaveTangent=0.8097, time=0.4671, value=0.7411}
          {arriveTangent=0.1287, interpMode=RCIM_Cubic, leaveTangent=0.1287, tangentMode=RCTM_User, time=1, value=0.9859}
         ]}}})))
   &DynamicMaterialParameters(
    Index 0 Param 1=&FloatFromCurve(FloatCurve=DataInterface() {propertyValues {Curve {keys [
         {arriveTangent=-1.3562, interpMode=RCIM_Cubic, leaveTangent=-1.3562, tangentMode=RCTM_User, time=0, value=2}
         {arriveTangent=-3.7058, interpMode=RCIM_Cubic, leaveTangent=-3.7058, tangentMode=RCTM_User, time=1, value=-0.5}
        ]}}})
    Index 0 Param 2=0.3
    Index 0 Param 3=3
    Index 0 Param 4=1
    Module.Param0WriteEnabled=true
    Module.Param1WriteEnabled=true
    Module.Param2WriteEnabled=true
    Module.Param3WriteEnabled=true
   )
   &ScaleColor(
    Scale Alpha=&FloatFromCurve(FloatCurve=DataInterface() {propertyValues {Curve {keys [<at 0 1 RCIM_Cubic>, {arriveTangent=-1.8559, leaveTangent=-1.8559, time=1, value=0}]}}})
    Scale RGB=&VectorFromCurve(VectorCurve=DataInterface() {propertyValues {XCurve {keys [<at 0 1>,<at 1 1>]}, YCurve {keys [<at 0 1>,<at 1 1>]}, ZCurve {keys [<at 0 1>,<at 1 1>]}}})
    ScaleA=true
    ScaleRGB=true
   )
  ]
  rendererClasses [&NiagaraMeshRendererProperties]
  renderers [{rendererClass=&NiagaraMeshRendererProperties, rendererIndex=0}]
 }
 Sparklies {
  bEnabled=false
  simTarget=GPUComputeSim
  emitterUpdateScript [
   &EmitterState(Life Cycle Mode=::NewEnumerator1, Loop Behavior=::NewEnumerator1, Loop Duration=0.22, UseLoopDelay=true)
   &SpawnBurst_Instantaneous(Spawn Count=50, Spawn Time=0.05, Use Spawn Probability=false, UseLoopCountLimit=false)
   &SpawnRate(
    SpawnRate=&Multiply_Float(
     A=1600
     B=&FloatFromCurve(
      FloatCurve=DataInterface() {propertyValues {Curve {keys [
          <at 0 1 RCIM_Cubic>
          {arriveTangent=-0.2543, interpMode=RCIM_Cubic, leaveTangent=-0.2543, time=1.2828, value=1.6362}
          {arriveTangent=-0.3703, interpMode=RCIM_Cubic, leaveTangent=-0.3703, time=2.4472, value=0.3776}
          {arriveTangent=-0.11, interpMode=RCIM_Cubic, leaveTangent=-0.11, time=3.6302, value=0.7671}
          {arriveTangent=0.2907, interpMode=RCIM_Cubic, leaveTangent=0.2907, time=5.034, value=0.0931}
          {arriveTangent=0.3058, interpMode=RCIM_Cubic, leaveTangent=0.3058, time=6.8459, value=1.702}
          <at 8 1>
         ]}}}
      CurveIndex=$Emitter.NormalizedLoopAge
     )
    )
    Use Spawn Probability=false
   )
  ]
  particleSpawnScript [
   /SetVariables(Particles.ShapeLocationOffset:Vector3f=&RandomRangeVector(Minimum=<-1000 0 0>, Maximum=<-10000 0 0>, Evaluation Type=::NewEnumerator1)) {isSetParameters=true}
   /SetVariables(Particles.ShapeLocationOffsetSpace:ENiagaraCoordinateSpace=::Local) {isSetParameters=true}
   /SetVariables(Particles.SpriteRotationRate=&RandomRangeFloat(Minimum=-1, Maximum=1, Evaluation Type=::NewEnumerator1)) {isSetParameters=true}
   &InitializeParticle(
    Lifetime=&RandomRangeFloat(Minimum=0.6, Maximum=3, Evaluation Type=::NewEnumerator1)
    Sprite Rotation Angle=&RandomRangeFloat(Minimum=0, Maximum=1, Evaluation Type=::NewEnumerator1)
    Sprite Rotation Mode=::NewEnumerator1
    Sprite Size=&Vector2DFromFloat(Value=&RandomRangeFloat(Minimum=1200, Maximum=2500, Evaluation Type=::NewEnumerator1))
    Sprite Size Mode=::NewEnumerator4
    UsePositionOffset=false
    Write Material Random=true
   )
   &AddVelocity(Velocity=&RandomRangeVector(Minimum=<4000 0 20000>, Maximum=<160000 0 30000>, Evaluation Type=::NewEnumerator1))
   &Color(Color=&MakeLinearColorFromVectorAndFloat("Vector (RGB)"=<50 1.25 1.25>, "Float (Alpha)"=1))
   &ShapeLocation(
    Hemisphere Distribution=<0.5 0.5>
    MaskSpawnGroup=false
    Offset=$Particles.ShapeLocationOffset
    Offset Coordinate Space=$Particles.ShapeLocationOffsetSpace
    Offset Mode=::NewEnumerator0
    Override Seed=false
    Rotation Mode=::NewEnumerator2
    Sphere Radius=&RandomRangeFloat(Minimum=200, Maximum=400, Evaluation Type=::NewEnumerator1)
    UseOwnerScale=true
    Yaw / Pitch / Roll=<0.75 0 0>
   )
   &FixShapeVector(
    Offset:Vector3f=$Particles.ShapeLocationOffset
    Offset Coordinate Space:ENiagaraCoordinateSpace=$Particles.ShapeLocationOffsetSpace
    Offset Mode:ENiagara_OffsetMode=::NewEnumerator0
   )
   &AddVelocity(Velocity=&Multiply_VectorByFloat(
     Vector=$Output.FixShapeVector.ShapeVector
     Float=&RandomRangeFloat(Minimum=50, Maximum=180, Evaluation Type=::NewEnumerator1)
    )) {module=AddVelocity001}
  ]
  particleUpdateScript [
   &ParticleState()
   &SpriteRotationRate(Rotation Rate=&AngleConversion(
     Angle:NiagaraFloat=$Particles.SpriteRotationRate
     Angle Input:ENiagara_AngleInput=::NewEnumerator1
     Angle Output:ENiagara_AngleInput=::NewEnumerator0
    ))
   &ScaleSpriteSize(
    Scale Sprite Size Mode=::NewEnumerator0
    Uniform Scale Factor=&FloatFromCurve(FloatCurve=DataInterface() {propertyValues {Curve {keys [
         {arriveTangent=-0.4739, time=-0.0044, value=1}
         {arriveTangent=2.5716, interpMode=RCIM_Cubic, leaveTangent=2.5716, time=0.157, value=0.5261}
         {arriveTangent=-4.5442, interpMode=RCIM_Cubic, leaveTangent=-4.5442, time=0.1659, value=1.438}
         {arriveTangent=-5.3292, interpMode=RCIM_Cubic, leaveTangent=-5.3292, time=0.1835, value=0.4056}
         {arriveTangent=1.8634, interpMode=RCIM_Cubic, leaveTangent=1.8634, time=0.3855, value=0.268}
         {arriveTangent=-0.0008, interpMode=RCIM_Cubic, leaveTangent=-0.0008, time=0.4129, value=0.833}
         {arriveTangent=-2.1307, interpMode=RCIM_Cubic, leaveTangent=-2.1307, time=0.4349, value=0.2679}
         {arriveTangent=3.0903, interpMode=RCIM_Cubic, leaveTangent=3.0903, time=0.689, value=0.2446}
         {arriveTangent=-0.8294, interpMode=RCIM_Cubic, leaveTangent=-0.8294, time=0.7034, value=1.0975}
         {arriveTangent=-1.107, interpMode=RCIM_Cubic, leaveTangent=-1.107, tangentMode=RCTM_User, time=0.7161, value=0.2222}
         <at 1 0>
        ]}}})
   )
   &Drag(
    Drag=&RandomRangeFloat(Minimum=0.3, Maximum=1.5, Evaluation Type=::NewEnumerator1)
    Use Linear Drag=true
    Use Rotational Drag=true
   )
   &SolveForcesAndVelocity(Clamp Velocity=false, Limit Acceleration=false)
  ]
  rendererClasses [&NiagaraSpriteRendererProperties]
  renderers [{rendererClass=&NiagaraSpriteRendererProperties}]
 }
 SparkliesOpposite {
  bEnabled=false
  simTarget=GPUComputeSim
  emitterUpdateScript [
   &EmitterState(Life Cycle Mode=::NewEnumerator1, Loop Behavior=::NewEnumerator1, Loop Delay=0.07, Loop Duration=0.22, UseLoopDelay=true)
   &SpawnBurst_Instantaneous(Spawn Count=25, Use Spawn Probability=false, UseLoopCountLimit=false)
   &SpawnRate(
    SpawnRate=&Multiply_Float(
     A=600
     B=&FloatFromCurve(
      FloatCurve=DataInterface() {propertyValues {Curve {keys [
          <at 0 1 RCIM_Cubic>
          {arriveTangent=-0.2543, interpMode=RCIM_Cubic, leaveTangent=-0.2543, time=1.2828, value=1.6362}
          {arriveTangent=-0.3703, interpMode=RCIM_Cubic, leaveTangent=-0.3703, time=2.4472, value=0.3776}
          {arriveTangent=-0.11, interpMode=RCIM_Cubic, leaveTangent=-0.11, time=3.6302, value=0.7671}
          {arriveTangent=0.2907, interpMode=RCIM_Cubic, leaveTangent=0.2907, time=5.034, value=0.0931}
          {arriveTangent=0.3058, interpMode=RCIM_Cubic, leaveTangent=0.3058, time=6.8459, value=1.702}
          <at 8 1>
         ]}}}
      CurveIndex=$Emitter.NormalizedLoopAge
     )
    )
    Use Spawn Probability=false
   )
  ]
  particleSpawnScript [
   /SetVariables(Particles.ShapeLocationOffset:Vector3f=&RandomRangeVector(Minimum=<-30000 0 0>, Maximum=<-20000 0 0>, Evaluation Type=::NewEnumerator1)) {isSetParameters=true}
   /SetVariables(Particles.ShapeLocationOffsetSpace:ENiagaraCoordinateSpace=::Local) {isSetParameters=true}
   /SetVariables(Particles.SpriteRotationRate=&RandomRangeFloat(Minimum=-1, Maximum=1, Evaluation Type=::NewEnumerator1)) {isSetParameters=true}
   &InitializeParticle(
    Lifetime=&RandomRangeFloat(Minimum=0.6, Maximum=3, Evaluation Type=::NewEnumerator1)
    Sprite Rotation Angle=&RandomRangeFloat(Minimum=0, Maximum=1, Evaluation Type=::NewEnumerator1)
    Sprite Rotation Mode=::NewEnumerator1
    Sprite Size=&Vector2DFromFloat(Value=&RandomRangeFloat(Minimum=800, Maximum=1500, Evaluation Type=::NewEnumerator1))
    Sprite Size Mode=::NewEnumerator4
    UsePositionOffset=false
    Write Material Random=true
   )
   &AddVelocity(Velocity=&RandomRangeVector(Minimum=<-55000 0 0>, Maximum=<-10000 0 10000>, Evaluation Type=::NewEnumerator1))
   &Color(Color=&MakeLinearColorFromVectorAndFloat("Vector (RGB)"=<50 1.25 1.25>, "Float (Alpha)"=1))
   &ShapeLocation(
    Hemisphere Distribution=<0.5 1>
    MaskSpawnGroup=false
    Non Uniform Scale=<1 1 0>
    Offset=$Particles.ShapeLocationOffset
    Offset Coordinate Space=$Particles.ShapeLocationOffsetSpace
    Offset Mode=::NewEnumerator0
    Override Seed=false
    Rotation Mode=::NewEnumerator2
    Sphere Radius=&RandomRangeFloat(Minimum=200, Maximum=400, Evaluation Type=::NewEnumerator1)
    UseOwnerScale=true
    Yaw / Pitch / Roll=<0.25 0 0>
   )
   &FixShapeVector(
    Offset:Vector3f=$Particles.ShapeLocationOffset
    Offset Coordinate Space:ENiagaraCoordinateSpace=$Particles.ShapeLocationOffsetSpace
    Offset Mode:ENiagara_OffsetMode=::NewEnumerator0
   )
   &AddVelocity(Velocity=&Multiply_VectorByFloat(
     Vector=$Output.FixShapeVector.ShapeVector
     Float=&RandomRangeFloat(Minimum=50, Maximum=150, Evaluation Type=::NewEnumerator1)
    )) {module=AddVelocity001}
  ]
  particleUpdateScript [
   &ParticleState()
   &SpriteRotationRate(Rotation Rate=&AngleConversion(
     Angle:NiagaraFloat=$Particles.SpriteRotationRate
     Angle Input:ENiagara_AngleInput=::NewEnumerator1
     Angle Output:ENiagara_AngleInput=::NewEnumerator0
    ))
   &ScaleSpriteSize(
    Scale Sprite Size Mode=::NewEnumerator0
    Uniform Scale Factor=&FloatFromCurve(FloatCurve=DataInterface() {propertyValues {Curve {keys [
         {arriveTangent=-0.4739, time=-0.0044, value=1}
         {arriveTangent=2.5716, interpMode=RCIM_Cubic, leaveTangent=2.5716, time=0.157, value=0.5261}
         {arriveTangent=-4.5442, interpMode=RCIM_Cubic, leaveTangent=-4.5442, time=0.1659, value=1.438}
         {arriveTangent=-5.3292, interpMode=RCIM_Cubic, leaveTangent=-5.3292, time=0.1835, value=0.4056}
         {arriveTangent=1.8634, interpMode=RCIM_Cubic, leaveTangent=1.8634, time=0.3855, value=0.268}
         {arriveTangent=-0.0008, interpMode=RCIM_Cubic, leaveTangent=-0.0008, time=0.4129, value=0.833}
         {arriveTangent=-2.1307, interpMode=RCIM_Cubic, leaveTangent=-2.1307, time=0.4349, value=0.2679}
         {arriveTangent=3.0903, interpMode=RCIM_Cubic, leaveTangent=3.0903, time=0.689, value=0.2446}
         {arriveTangent=-0.8294, interpMode=RCIM_Cubic, leaveTangent=-0.8294, time=0.7034, value=1.0975}
         {arriveTangent=-1.107, interpMode=RCIM_Cubic, leaveTangent=-1.107, tangentMode=RCTM_User, time=0.7161, value=0.2222}
         <at 1 0>
        ]}}})
   )
   &Drag(
    Drag=&RandomRangeFloat(Minimum=1, Maximum=2.5, Evaluation Type=::NewEnumerator1)
    Use Linear Drag=true
    Use Rotational Drag=true
   )
   &SolveForcesAndVelocity(Clamp Velocity=false, Limit Acceleration=false)
  ]
  rendererClasses [&NiagaraSpriteRendererProperties]
  renderers [{rendererClass=&NiagaraSpriteRendererProperties}]
 }
 Sphere {
  bEnabled=false
  emitterUpdateScript [
   &EmitterState(Life Cycle Mode=::NewEnumerator1, Loop Behavior=::NewEnumerator1, Loop Delay=0.2, Loop Duration=1, UseLoopDelay=true)
   &SpawnBurst_Instantaneous(Use Spawn Probability=false, UseLoopCountLimit=false)
  ]
  particleSpawnScript [
   &InitializeParticle(Lifetime=2, Mesh Scale=<700 700 700>, Mesh Scale Mode=::NewEnumerator4, UsePositionOffset=false, Write Material Random=true)
   &Color(Color=&MakeLinearColorFromVectorAndFloat("Vector (RGB)"=<10 2 2>, "Float (Alpha)"=0.025))
  ]
  particleUpdateScript [
   &ParticleState()
   &ScaleMeshSize(Scale Factor=&VectorFromFloat(Value=&FloatFromCurve(FloatCurve=DataInterface() {propertyValues {Curve {keys [
          {arriveTangent=1.5009, interpMode=RCIM_Cubic, leaveTangent=1.5009, tangentMode=RCTM_User, time=0, value=0.1762}
          {arriveTangent=0.8097, interpMode=RCIM_Cubic, leaveTangent=0.8097, time=0.4876, value=0.6963}
          {arriveTangent=0.1287, interpMode=RCIM_Cubic, leaveTangent=0.1287, tangentMode=RCTM_User, time=1, value=0.9859}
         ]}}}))) {module=ScaleMeshSize001}
   &ScaleMeshSize(Scale Factor=<5 5 5>)
   &ScaleColor(
    Scale Alpha=&FloatFromCurve(FloatCurve=DataInterface() {propertyValues {Curve {keys [
         {arriveTangent=-0.2749, interpMode=RCIM_Cubic, leaveTangent=-0.2749, tangentMode=RCTM_User, time=0, value=1}
         {arriveTangent=-0.9418, leaveTangent=-0.9418, time=1, value=0}
        ]}}})
    Scale RGB=&VectorFromCurve(VectorCurve=DataInterface() {propertyValues {XCurve {keys [<at 0 1>,<at 1 1>]}, YCurve {keys [<at 0 1>,<at 1 1>]}, ZCurve {keys [<at 0 1>,<at 1 1>]}}})
    ScaleA=true
    ScaleRGB=true
   )
   &DynamicMaterialParameters(
    Index 0 Param 1=&FloatFromCurve(FloatCurve=DataInterface() {propertyValues {Curve {keys [<at 0.5 0.95>,<at 1 1>]}}})
    Index 0 Param 2=&FloatFromCurve(FloatCurve=DataInterface() {propertyValues {Curve {keys [
         <at 0 1 RCIM_Cubic>
         {arriveTangent=-2, interpMode=RCIM_Cubic, leaveTangent=-2, time=0.5358, value=0.6834}
         {arriveTangent=-1.5094, interpMode=RCIM_Cubic, leaveTangent=-1.5094, tangentMode=RCTM_User, time=1, value=-1}
        ]}}})
    Index 0 Param 3=1
    Index 0 Param 4=1
    Module.Param0WriteEnabled=true
    Module.Param1WriteEnabled=true
    Module.Param2WriteEnabled=true
    Module.Param3WriteEnabled=true
   )
  ]
  rendererClasses [&NiagaraMeshRendererProperties]
  renderers [{rendererClass=&NiagaraMeshRendererProperties, rendererIndex=0}]
 }
}
summary {
 systemName=NS_DCN_Crack_Grow_Burst
 emitters [
  {bEnabled=false, emitterName=Corona, rendererClasses [&NiagaraSpriteRendererProperties]}
  {bEnabled=false, emitterName=Sphere, rendererClasses [&NiagaraMeshRendererProperties]}
  {bEnabled=false, emitterName=Shockwave, rendererClasses [&NiagaraMeshRendererProperties]}
  {bEnabled=false, emitterName=Frags1, rendererClasses [&NiagaraMeshRendererProperties]}
  {bEnabled=false, emitterName=Frags2, rendererClasses [&NiagaraMeshRendererProperties]}
  {bEnabled=false, emitterName=Frags3, rendererClasses [&NiagaraMeshRendererProperties]}
  {bEnabled=false, emitterName=Sparklies, simTarget=GPUComputeSim, rendererClasses [&NiagaraSpriteRendererProperties]}
  {bEnabled=false, emitterName=SparkliesOpposite, simTarget=GPUComputeSim, rendererClasses [&NiagaraSpriteRendererProperties]}
  {bEnabled=false, emitterName=RoundFlare, rendererClasses [&NiagaraSpriteRendererProperties]}
  {
   bEnabled=false
   emitterName=Minimal001
   rendererClasses [&NiagaraSpriteRendererProperties, &NiagaraMeshRendererProperties]
  }
  {bEnabled=false, emitterName=Minimal002, rendererClasses [&NiagaraSpriteRendererProperties]}
 ]
 userVariables [User.Stage {type=NiagaraFloat, defaultValue=1, description=""}]
}
